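# Translucent, unlit, two-sided Niagara aura material for the DualParadoxGold pickaxe.
# Panning noise distorts UVs, which are power-squeezed, offset by particle Random and panned across a mask texture.
# Mask channels drive a two-color emissive blend, scaled by time of day and particle color, and the opacity.
# Opacity adds a U-gradient cutoff, depth fade and view-angle fade, whose exponent varies with local-Z camera alignment and particle AngleFadeMult.
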Material /BRCosmetics/Weapons/FORT_Melee/Pickaxe_DualParadoxGold/FX/M_Pickaxe_DualParadoxGold_Aura
graph {
 properties {
  bUsedWithNiagaraMeshParticles=true
  bUsedWithNiagaraSprites=true
  blendMode=BLEND_Translucent
  pixelDepthOffsetMode=PDOM_Legacy
  shadingModel=MSM_Unlit
  twoSided=true
  usedShadingModels=Unlit
 }
 nodes [
  DynamicParameter() {defaultValue=<rgba 0 1 1 1>, paramNames [Random, AngleFadeMult, "", ""]}
  n1 = AppendVector(A=DynamicParameter() {defaultValue=<rgba 0 1 1 1>, paramNames [Random, AngleFadeMult, "", ""]}.Random, B=Constant())
  n2 = TextureCoordinate()
  n3 = VectorParameter() {group=Distortion, parameterName=Distortion UV, defaultValue=<rgba 8 0.1 0 1>}
  n4 = AppendVector(A=@n3.R, B=@n3.R)
  n5 = TextureSample(UVs=Panner(Coordinate=Multiply(A=@n2, B=@n4), Speed=AppendVector(A=@n3.B, B=@n3.B))) {
   samplerType=SAMPLERTYPE_LinearGrayscale
   texture=/Game/Characters/Player/Male/Medium/Bodies/M_MED_Citadel/FX/T_Citadel_Noise_Dissolve_Handpaint
  }
  n6 = ScalarParameter() {defaultValue=0.06, group=Distortion, parameterName=Distortion Intensity}
  n7 = MaterialFunctionCall(Float2=Add(A=TextureCoordinate(), B=Multiply(A=@n5.R, B=@n6))) {materialFunction=&BreakOutFloat2Components}
  n8 = ScalarParameter() {defaultValue=1.5, group=Mask, parameterName=Mask U Squeeze}
  n9 = ScalarParameter() {defaultValue=0.6, group=Mask, parameterName=Mask V Squeeze}
  n10 = VectorParameter() {group=Mask, parameterName=Mask UV, defaultValue=<rgba 0.5 1 -0.8 0>}
  n11 = Multiply(
   A=AppendVector(A=Power(Base=@n7.R, Exp=@n8) {constExponent=1.5}, B=Power(Base=@n7.G, Exp=@n9) {constExponent=0.7})
   B=AppendVector(A=@n10.R, B=@n10.R)
  )
  n12 = MaterialFunctionCall(Float2=Panner(Coordinate=Add(A=@n1, B=@n11), Speed=AppendVector(A=@n10.B, B=@n10.B))) {materialFunction=&BreakOutFloat2Components}
  n13 = MaterialFunctionCall(Float2=TextureCoordinate()) {materialFunction=&BreakOutFloat2Components}
  n14 = Multiply(A=Abs(None=Subtract(A=@n13.R) {constB=0.5})) {constB=2}
  n15 = ScalarParameter() {defaultValue=2, group=Mask, parameterName=Mask Edge Scale}
  n16 = Multiply(A=Saturate(None=Subtract(A=@n14, B=@n15)), B=Divide(B=Subtract(B=@n15)) {constA=0.5, constB=1}) {constB=2}
  n17 = Subtract(B=Multiply(A=Power(Base=@n16) {constExponent=4}) {constB=50})
  n18 = TextureSampleParameter2D(UVs=MaterialFunctionCall(X=@n12.R, Y=Multiply(A=@n12.G, B=@n17)) {materialFunction=&MakeFloat2}.Result) {group=Mask, parameterName=Mask Tex, texture=&/T_Pickaxe_DualParadoxGold_Mask_01}
  n19 = ChannelMaskParameter(None=@n18.RGBA) {group=Mask, maskChannel=Green, parameterName=Mask Emissive Channel, sortPriority=0, defaultValue=<rgba 0 1 0 0>}
  n20 = LinearInterpolate(
   A=VectorParameter() {group=Color, parameterName=Color A, defaultValue=<rgba 0.03 0 0.01 1>}.RGB
   B=VectorParameter() {group=Color, parameterName=Color B, defaultValue=<rgba 0.12 0 0.01 1>}.RGB
   Alpha=@n19.R
  )
  n21 = MaterialFunctionCall(
   Value_Morning=ScalarParameter() {defaultValue=1, group=TOD, parameterName=TOD_Morning}
   Value_Day=ScalarParameter() {defaultValue=1, group=TOD, parameterName=TOD_Day}
   Value_Evening=ScalarParameter() {defaultValue=1, group=TOD, parameterName=TOD_Evening}
   Value_Night=ScalarParameter() {defaultValue=1, group=TOD, parameterName=TOD_Night}
  ) {materialFunction=&Scalar_Time_of_Day}
  n22 = ParticleColor()
  MaterialFunctionCall(Emissive=Multiply(A=Multiply(A=@n20, B=@n21.Result), B=@n22.RGB)) {materialFunction=&MF_FX_EmissiveAdjustments}
  n23 = ChannelMaskParameter(None=@n18.RGBA) {group=Mask, maskChannel=Blue, parameterName=Mask Alpha Channel, sortPriority=0, defaultValue=<rgba 0 0 1 0>}
  n24 = ScalarParameter() {defaultValue=0.5, group=Mask, parameterName=Mask Alpha Power}
  n25 = TextureCoordinate()
  n26 = Multiply(
   A=Subtract(
    A=MaterialFunctionCall(Float2=@n25) {materialFunction=&BreakOutFloat2Components}.R
    B=ScalarParameter() {defaultValue=-0.05, group=Alpha, parameterName=Bottom Offset}
   ) {constB=0.4}
   B=ScalarParameter() {defaultValue=2, group=Alpha, parameterName=Bottom Hardness}
  ) {constB=2}
  n27 = Multiply(
   A=Multiply(A=Power(Base=@n23.R, Exp=@n24) {constExponent=0.5}, B=Saturate(None=@n26))
   B=ScalarParameter() {defaultValue=1.5, group=Alpha, parameterName=Alpha Mult}
  )
  n28 = ScalarParameter() {defaultValue=0.5, parameterName=FadeSensitivity A}
  n29 = ScalarParameter() {parameterName=FadeSensitivityB}
  n30 = Constant3Vector() {constant=<rgba 0 0 1 0>}
  n31 = Abs(None=DotProduct(A=Transform(None=@n30) {transformSourceType=TRANSFORMSOURCE_Local}, B=CameraVectorWS()))
  n32 = ScalarParameter() {defaultValue=0.5, parameterName=FadegradientOffset}
  n33 = MaterialFunctionCall(In=Saturate(None=Add(A=@n31, B=@n32)), Contrast=ScalarParameter() {defaultValue=0.5, parameterName=Fade gradient Contrast}) {materialFunction=&CheapContrast}
  n34 = Power(
   Base=MaterialFunctionCall("FadeSensitivity (optional)"=ScalarParameter() {defaultValue=0.2, group=Alpha, parameterName=AngleFade}) {materialFunction=&FadeBasedOnViewAngle}.Result
   Exp=Multiply(
    A=LinearInterpolate(A=@n28, B=@n29, Alpha=@n33.Result)
    B=DynamicParameter() {defaultValue=<rgba 0 1 1 1>, paramNames [Random, AngleFadeMult, "", ""]}.AngleFadeMult
   )
  ) {constExponent=1}
  Multiply(
   A=@n22.A
   B=DepthFade(Opacity=Multiply(A=@n27, B=@n34), FadeDistance=ScalarParameter() {defaultValue=20, group=Alpha, parameterName=DepthFade})
  )
 ]
}
