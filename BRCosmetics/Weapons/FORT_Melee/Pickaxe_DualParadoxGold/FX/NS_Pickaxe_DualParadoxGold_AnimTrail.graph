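# Niagara weapon swing trail for the DualParadoxGold pickaxe, drawn as two animation-trail ribbons anchored to skeletal mesh bones or sockets.
# One emitter makes a wider, longer-lived ribbon that shifts from pinkish red to dark red and tapers with age.
# The other uses a different second bone or socket for a narrower, brief violet ribbon with a tenfold RGB boost.
# Both fade out late in life and dim at night through a time-of-day scalar.
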NiagaraSystem /BRCosmetics/Weapons/FORT_Melee/Pickaxe_DualParadoxGold/FX/NS_Pickaxe_DualParadoxGold_AnimTrail
dependencies {
 usedDataInterfaces [&NiagaraDataInterfaceColorCurve, &NiagaraDataInterfaceCurve]
 usedDynamicInputs []
 usedModules [&SystemState]
}
emitters {
 FN_AnimTrail_V5 {
  emitterUpdateScript [
   &EmitterState()
   &FN_SpawnAnimTrail_v2(AnimTrailWidthMode=::FromSecond)
  ]
  particleSpawnScript [
   /Game/Effects/Niagara/Modules/AnimTrail/FN_InitializeAnimTrail/FN_InitializeAnimTrailBoneIndecies(Second Bone or Socket Index=1)
   /Game/Effects/Niagara/Modules/AnimTrail/FN_InitializeAnimTrail/FN_InitializeAnimTrail(AnimTrailWidth=1, Color=<rgba 0.05 0 0.0167 1.5>, Lifetime=0.2)
  ]
  particleUpdateScript [
   &SolveForcesAndVelocity(Clamp Velocity=false, Limit Acceleration=false)
   &FN_UpdateAnimTrail_v2()
   &Color(Color=&ColorFromCurve(ColorCurve=DataInterface() {propertyValues {
       AlphaCurve {keys [<at 0 1.5 RCIM_Cubic>]}
       BlueCurve {keys [<at 0.0032 0.125>,<at 0.4057 0.0083 RCIM_Cubic>]}
       GreenCurve {keys [<at 0.0032 0.05>,<at 0.4057 0 RCIM_Cubic>]}
       RedCurve {keys [<at 0.0032 0.5>,<at 0.4057 0.05 RCIM_Cubic>]}
      }}))
   &ScaleColor(
    Color Value To Scale=$Particles.Color
    Scale Alpha=&FloatFromCurve(FloatCurve=DataInterface() {propertyValues {Curve {keys [
         {arriveTangent=-5.3881, interpMode=RCIM_Cubic, leaveTangent=-5.3881, tangentMode=RCTM_User, time=0.7, value=1}
         {arriveTangent=-1.2106, interpMode=RCIM_Cubic, leaveTangent=-1.2106, tangentMode=RCTM_User, time=1, value=0}
        ]}}})
    Scale RGB=&VectorFromFloat(Value=&MakeFloatFromVector(
      Vector=&VectorFromFloat(Value=&FNTimeOfDayScalar(Night Scale=0.7, Morning Scale=1, Evening Scale=1))
      Channel=::NewEnumerator0
     ))
    ScaleA=true
    ScaleRGB=true
   )
   &ScaleRibbonWidth(Ribbon Width Scale=&FloatFromCurve(FloatCurve=DataInterface() {propertyValues {Curve {keys [<at 0.5 1>,<at 1 0>]}}}))
  ]
  rendererClasses [&NiagaraRibbonRendererProperties]
  renderers [{rendererClass=&NiagaraRibbonRendererProperties}]
 }
 FN_AnimTrail_V5001 {
  emitterUpdateScript [
   &EmitterState()
   &FN_SpawnAnimTrail_v2(AnimTrailWidthMode=::FromSecond)
  ]
  particleSpawnScript [
   &FN_InitializeAnimTrailBoneIndecies(Second Bone or Socket Index=2)
   &FN_InitializeAnimTrail(AnimTrailWidth=0.75, Color=<rgba 0.7 0.3 1.5 0.2>, Lifetime=0.1)
  ]
  particleUpdateScript [
   &SolveForcesAndVelocity(Clamp Velocity=false, Limit Acceleration=false)
   &FN_UpdateAnimTrail_v2()
   &ScaleColor(
    Scale Alpha=&FloatFromCurve(FloatCurve=DataInterface() {propertyValues {Curve {keys [{arriveTangent=-5.6186, time=0.7, value=1}, {arriveTangent=-1.186, leaveTangent=-1.186, time=1, value=0}]}}})
    Scale RGB=&VectorFromFloat(Value=&Multiply_Float(
      A=10
      B=&MakeFloatFromVector(
       Vector=&VectorFromFloat(Value=&FNTimeOfDayScalar(Night Scale=0.35, Morning Scale=1, Evening Scale=1))
       Channel=::NewEnumerator0
      )
     ))
    ScaleA=true
    ScaleRGB=true
   )
  ]
  rendererClasses [&NiagaraRibbonRendererProperties]
  renderers [{rendererClass=&NiagaraRibbonRendererProperties}]
 }
}
summary {
 systemName=NS_Pickaxe_DualParadoxGold_AnimTrail
 emitters [
  {emitterName=FN_AnimTrail_V5, rendererClasses [&NiagaraRibbonRendererProperties]}
  {emitterName=FN_AnimTrail_V5001, rendererClasses [&NiagaraRibbonRendererProperties]}
 ]
 userVariables [
  User.RibbonIDOffset {type=NiagaraInt32}
  User.Secondary Color {type=LinearColor, description="", defaultValue {struct=LinearColor, value=<rgba 1 1 1 1>}}
  User.SkeletalMesh {
   type=NiagaraDataInterfaceSkeletalMesh
   defaultValue {
    struct=NiagaraExt_VariableValue_DataInterface
    value {dataInterface=&self:NiagaraDataInterfaceSkeletalMesh_0}
   }
  }
 ]
}
